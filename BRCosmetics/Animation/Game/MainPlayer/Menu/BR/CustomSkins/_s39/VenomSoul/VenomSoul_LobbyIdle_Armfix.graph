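# Level Sequence for the VenomSoul skin's lobby idle, apparently an authoring sequence for an arm fix, judging by its name.
# A FortPlayerMannequin binding has a spawn track, an identity transform and a skeletal animation section as the base motion.
# An MR_Biped_v01 Control Rig track in Layered mode sits over that animation, allowing additive corrections without replacing the original clip.
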LevelSequence /BRCosmetics/Animation/Game/MainPlayer/Menu/BR/CustomSkins/_s39/VenomSoul/VenomSoul_LobbyIdle_Armfix
bindings [FortPlayerMannequin {
  id=#1
  possessed_class=FortPlayerMannequin
  tracks [
   {type=MovieSceneSpawnTrack, sections [{type=MovieSceneSpawnSection, channels [None {type=MovieSceneScriptingBoolChannel, default=true}]}]}
   {
    type=MovieScene3DTransformTrack
    sections [{
      type=MovieScene3DTransformSection
      channels [
       Location.X {type=MovieSceneScriptingDoubleChannel, default=0}
       Location.Y {type=MovieSceneScriptingDoubleChannel, default=0}
       Location.Z {type=MovieSceneScriptingDoubleChannel, default=0}
       Rotation.X {type=MovieSceneScriptingDoubleChannel, default=0}
       Rotation.Y {type=MovieSceneScriptingDoubleChannel, default=0}
       Rotation.Z {type=MovieSceneScriptingDoubleChannel, default=0}
       Scale.X {type=MovieSceneScriptingDoubleChannel, default=1}
       Scale.Y {type=MovieSceneScriptingDoubleChannel, default=1}
       Scale.Z {type=MovieSceneScriptingDoubleChannel, default=1}
      ]
     }]
   }
   {type=MovieSceneSkeletalAnimationTrack, sections [{type=MovieSceneSkeletalAnimationSection, end_frame=320, start_frame=0}]}
   {type=MovieSceneControlRigParameterTrack, display_name="MR_Biped_v01 (Layered)", sections [{type=MovieSceneControlRigParameterSection}]}
  ]
 }]
display_rate {denominator=1, numerator=30}
tick_resolution {denominator=1, numerator=24000}
tracks []
